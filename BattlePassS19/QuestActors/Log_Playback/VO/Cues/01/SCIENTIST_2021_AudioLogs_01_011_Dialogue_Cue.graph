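# Sound cue for line 011 of the Scientist's 2021 audio log 01, part of the Battle Pass S19 quest log playback voice-over.
# A single dialogue player node at the root plays the line's Dialogue Wave, with the Scientist voice as speaker and no listed targets.
# The cue uses the shared S19 audio log attenuation and routes through the narrator VO sound class.
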SoundCue /BattlePassS19/QuestActors/Log_Playback/VO/Cues/01/SCIENTIST_2021_AudioLogs_01_011_Dialogue_Cue
root=DialoguePlayer_0
nodes [{
  id=DialoguePlayer_0
  properties {dialogueWaveParameter {
    dialogueWave=/BattlePassS19/QuestActors/Log_Playback/VO/Waves/SCIENTIST_2021_AudioLogs_01_011_Dialogue
    context {speaker=/BattlePassS19/Audio/VO/DV_Scientist, targets []}
   }}
 }]
properties {
 attenuationSettings=/BattlePassS19/QuestActors/Log_Playback/VO/S19_AudioLogs_Att
 maxDistance=2097152
 soundClassObject=/Game/Sounds/FortSoundClassesAndModes/Fort_VO_Narrator
 volumeMultiplier=1
}
